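# Unlit translucent retro skydiving contrail material for Niagara ribbons and mesh particles, with warm default colors.
# Quantized, panning UVs feed a sine hash that drives blinking blocks, layered with scrolling grid, line and sweep textures.
# Opacity combines blocks or a gradient break shape with scanlines and camera fade, or uses a plain glow via Use_BlockTex.
# Dynamic parameters control hash seed, scroll speed, blink rate, break blend and TOD desaturation.
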Material /BRCosmetics/Athena/Effects/Particles/Skydiving/NS_Contrail_Dive_Retro/M_Dive_Retro_Trail
graph {
 properties {
  bUsedWithNiagaraMeshParticles=true
  bUsedWithNiagaraRibbons=true
  blendMode=BLEND_Translucent
  pixelDepthOffsetMode=PDOM_Legacy
  shadingModel=MSM_Unlit
  usedShadingModels=Unlit
 }
 nodes [
  n1 = ParticleColor()
  n2 = DepthFade(Opacity=@n1.A, FadeDistance=ScalarParameter() {defaultValue=5, parameterName=DepthFade})
  n3 = ShadingPathSwitch(Default=@n2, Deferred=@n2, Forward=@n1.A, Mobile=@n1.A)
  n4 = OneMinus(None=ComponentMask(None=TextureCoordinate()) {r=true})
  n5 = ComponentMask(None=TextureCoordinate()) {g=true}
  n6 = OneMinus(None=@n5)
  n7 = Multiply(A=@n4, B=Multiply(A=Multiply(A=@n5, B=@n6), B=ScalarParameter() {defaultValue=3, parameterName=GlowScale A}))
  n8 = Clamp(None=Multiply(A=OneMinus(None=@n4), B=ScalarParameter() {defaultValue=3, parameterName=Scale A}))
  n9 = TextureCoordinate()
  n10 = ScalarParameter() {defaultValue=-0.4, parameterName=11Pan U}
  n11 = DynamicParameter() {defaultValue=<rgba 1 0 1 1>, paramNames [Random,TOD,Speed,Use_break]}
  n12 = Multiply(
   A=Time() {bOverride_Period=true, period=10000}
   B=AppendVector(A=Multiply(A=@n10, B=@n11.Speed), B=ScalarParameter() {parameterName=11Pan V})
  )
  n13 = Constant() {r=30}
  n14 = Constant() {r=3}
  n15 = Floor(None=Multiply(A=Add(A=@n9, B=@n12), B=AppendVector(A=@n13, B=@n14)))
  n16 = Constant() {r=5}
  n17 = AppendVector(A=ScalarParameter() {defaultValue=3, parameterName=DotA}, B=ScalarParameter() {defaultValue=1, parameterName=DotB})
  n18 = AppendVector(A=ScalarParameter() {defaultValue=4, parameterName=DotA2}, B=ScalarParameter() {defaultValue=2, parameterName=DotB2})
  n19 = DynamicParameter() {defaultValue=<rgba 1 0 1 1>, paramNames [Random,TOD,Speed,Use_break]}
  n20 = Sine(None=DotProduct(A=Divide(A=@n15, B=@n16), B=LinearInterpolate(A=@n17, B=@n18, Alpha=@n19.Random)))
  n21 = ScalarParameter() {defaultValue=400000, parameterName=mmmm}
  n22 = Add(A=Time(), B=Frac(None=Multiply(A=@n20, B=@n21)))
  n23 = Ceil(None=LinearInterpolate(
    A=Sine(None=@n22) {period=0.2}
    B=Sine(None=@n22) {period=0.4}
    Alpha=DynamicParameter() {parameterIndex=1, paramNames [Sine_Speed,Param2,Param3,Param4]}.Sine_Speed
   ))
  n24 = TextureCoordinate()
  n25 = Constant2Vector() {r=-0.45}
  n26 = Ceil(None=ComponentMask(None=Add(A=@n24, B=@n25)) {r=true})
  n27 = TextureSampleParameter2D(UVs=Panner(Coordinate=TextureCoordinate() {uTiling=2, vTiling=0.38}) {speedX=-2}) {parameterName=LineTex001, samplerType=SAMPLERTYPE_LinearGrayscale, texture=/BRCosmetics/Weapons/FORT_Melee/Pickaxe_CatSpace/FX/T_SquareTile}
  n28 = ComponentMask(None=Add(A=TextureCoordinate(), B=Constant2Vector() {r=-0.3})) {r=true}
  n29 = Add(A=Multiply(A=OneMinus(None=@n26), B=LinearInterpolate(Alpha=@n27.R)), B=OneMinus(None=Ceil(None=@n28)))
  n30 = OneMinus(None=ComponentMask(None=TextureCoordinate()) {r=true})
  n31 = Multiply(A=Power(Base=Add(A=@n30) {constB=0.45}) {constExponent=8}) {constB=3}
  n32 = LinearInterpolate(
   A=Multiply(A=@n3, B=Add(A=@n23, B=@n29))
   B=Multiply(A=@n3, B=Saturate(None=@n31))
   Alpha=DynamicParameter() {defaultValue=<rgba 1 0 1 1>, paramNames [Random,TOD,Speed,Use_break]}.Use_break
  )
  n33 = TextureCoordinate()
  n34 = Saturate(None=Multiply(A=ComponentMask(None=@n33) {r=true}) {constB=8})
  n35 = Multiply(
   A=Multiply(A=@n34, B=@n34)
   B=MaterialFunctionCall(
    Fade Length=ScalarParameter() {defaultValue=50, parameterName="Fade Length (S)"}
    Fade Offset=ScalarParameter() {defaultValue=25, parameterName="Fade Offset (S)"}
   ) {materialFunction=&CameraDepthFade}.Result
  )
  n36 = ScalarParameter() {defaultValue=50, group=Scanlines, parameterName=Scanline Amount}
  n37 = ScalarParameter() {defaultValue=-0.15, group=Scanlines, parameterName=ScanlineSize}
  n38 = ScalarParameter() {defaultValue=0.5, group=Scanlines, parameterName=Scanlines Dark}
  n39 = ScalarParameter() {defaultValue=1, group=Scanlines, parameterName=Scanlines Light}
  n40 = ScalarParameter() {defaultValue=1, parameterName=Use_BlockTex}
  LinearInterpolate(
   A=Multiply(A=@n3, B=Multiply(A=@n7, B=@n8))
   B=Multiply(
    A=Multiply(A=@n32, B=@n35)
    B=MaterialFunctionCall!(
     Number of Scanlines=@n36
     Scanline Speed
     "Scanlines Thickness (Bias)"=@n37
     Scanlines Dark Value=@n38
     Scanlines Light Value=@n39
     Block Opacity
     UVs
     Time
     Block Repeat
    ) {materialFunction=/Game/Effects/Fort_Effects/Effects/Atlas/Materials/MF_Atlas_Scanlines_01}.Result
   )
   Alpha=@n40
  )
  n41 = VectorParameter() {parameterName=00000, defaultValue=<rgba 1.2 0.3133 0.0236 1>}
  n42 = OneMinus(None=@n23)
  n43 = Multiply(
   A=TextureCoordinate()
   B=AppendVector(A=ScalarParameter() {defaultValue=3, parameterName=Grid Texture U}, B=ScalarParameter() {defaultValue=0.3, parameterName=Grid Texture V})
  )
  n44 = ScalarParameter() {defaultValue=-1.2, parameterName=Grid Texture Pan U}
  n45 = DynamicParameter() {defaultValue=<rgba 1 0 1 1>, paramNames [Random,TOD,Speed,Use_break]}
  n46 = Multiply(
   A=Time() {bOverride_Period=true, period=10000}
   B=AppendVector(A=Multiply(A=@n44, B=@n45.Speed), B=ScalarParameter() {parameterName=Grid Texture Pan V})
  )
  n47 = ChannelMaskParameter(None=TextureSampleParameter2D(UVs=Add(A=@n43, B=@n46)) {parameterName=Grid Texture, texture=/BRCosmetics/Athena/Effects/Particles/Skydiving/NS_Contrail_80sRibbon/Textures/T_Grid}.RGBA) {maskChannel=Blue, parameterName=Grid Texture Channel, defaultValue=<rgba 0 0 1 0>}
  n48 = Multiply(A=ScalarParameter() {defaultValue=-1, parameterName=LinePan U}, B=@n45.Speed)
  n49 = ScalarParameter() {parameterName=LinePan V}
  n50 = TextureSampleParameter2D(UVs=Panner(Coordinate=TextureCoordinate() {uTiling=3, vTiling=0.38}, Speed=AppendVector(A=@n48, B=@n49)) {speedY=-1.5}) {parameterName=LineTex, samplerType=SAMPLERTYPE_LinearGrayscale, texture=&T_SquareTile}
  n51 = Add(
   A=Add(A=Multiply(A=@n41.RGB, B=@n42), B=VectorParameter() {parameterName=Color Low, defaultValue=<rgba 0.4 0.1608 0.0333 1>}.RGB)
   B=Multiply(A=VectorParameter() {parameterName=GridColor, defaultValue=<rgba 3 1.6456 0.1379 1>}.RGB, B=Multiply(A=@n47.R, B=@n50.R))
  )
  n52 = VectorParameter() {parameterName=ColorLine, defaultValue=<rgba 1.5 1.2558 0.1484 1>}
  n53 = MaterialFunctionCall(XY=TextureCoordinate() {uTiling=1.2}) {materialFunction=&Swizzle}
  n54 = AppendVector(A=ScalarParameter() {parameterName=SweepPan U}, B=Multiply(A=ScalarParameter() {defaultValue=-1, parameterName=SweepPan V}, B=@n45.Speed))
  n55 = Power(Base=TextureSampleParameter2D(UVs=Panner(Coordinate=@n53.YX, Speed=@n54) {speedY=-1.5}) {parameterName=Tex_Sweep, texture=/Game/Global/Textures/Default/Masks_and_Gradients/T_VertGradient_02_sRGB}.R) {constExponent=3}
  n56 = Saturate(None=Multiply(A=Multiply(A=@n55)) {constB=5})
  n57 = Desaturation(None=Multiply(A=@n51, B=Multiply(A=@n52.RGB, B=@n56)), Fraction=Constant())
  n58 = VectorParameter() {parameterName=ColorBlock, defaultValue=<rgba 0.2604 0.1098 0.0277 1>}
  n59 = Multiply(A=Add(A=Add(A=@n51, B=@n57), B=Multiply(A=@n58.RGB, B=@n23)), B=@n1.RGB)
  MaterialFunctionCall(Emissive=Desaturation(
    None=LinearInterpolate(A=@n1.RGB, B=@n59, Alpha=@n40)
    Fraction=DynamicParameter() {defaultValue=<rgba 1 0 1 1>, paramNames [Random,TOD,Speed,Use_break]}.TOD
   )) {materialFunction=&MF_FX_EmissiveAdjustments}
 ]
}
