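# Niagara skydiving dive contrail cosmetic with two wavy ribbons, socket-attached glow sprites and two color variants of drifting plane particles around the character.
# Ribbons spread five particles upward from skeletal mesh sockets, bending them with age-driven sine and cosine offsets.
# Mesh-and-sprite plane emitters send location and death events, alongside short-lived Line and Die sprite emitters without their own spawn modules.
# Glide lean, front-end preview, time of day and platform adjust the look.
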NiagaraSystem /BRCosmetics/Athena/Effects/Particles/Skydiving/NS_Contrail_Dive_Retro/NS_Contrail_Dive_Retro
dependencies {
 usedDataInterfaces [
  &NiagaraDataInterfaceCurve
  &NiagaraDataInterfacePlatformSet
  &NiagaraDataInterfaceColorCurve
  &NiagaraDataInterfaceArrayFloat3
 ]
 usedDynamicInputs []
 usedModules [
  &SystemState
  &InterpolateOverTime
  &self:SystemSpawnScript.SystemScriptSource.SystemScriptGraph.NiagaraNodeAssignment_0.SetVariables_#1
  &self:SystemSpawnScript.SystemScriptSource.SystemScriptGraph.NiagaraNodeAssignment_1.SetVariables_#2
 ]
}
emitters {
 Blue_Die {
  emitterUpdateScript [&EmitterState(Life Cycle Mode=::NewEnumerator1, Loop Duration=1, UseLoopDelay=false)]
  particleSpawnScript [
   &InitializeParticle(
    Color=<rgba 5 0.04 0.02 1>
    Lifetime Max=0.2
    Lifetime Min=0.15
    Lifetime Mode=::NewEnumerator1
    Sprite Rotation Mode=::NewEnumerator2
    Sprite Size Mode=::NewEnumerator3
    Uniform Sprite Size=100
    UsePositionOffset=false
    Write Material Random=true
   )
   &SubUVAnimation(End Frame=3, SubUV Animation Mode:ENiagara_SubUVLookupMode=::NewEnumerator2)
  ]
  particleUpdateScript [
   &ParticleState()
   &ScaleColor(
    Linear Color Curve=DataInterface() {propertyValues {
      AlphaCurve {keys [<at 0.3624 1>,<at 0.8992 0>]}
      BlueCurve {keys [<at 0.1978 1>]}
      GreenCurve {keys [<at 0.1978 1>]}
      RedCurve {keys [<at 0.1978 1>]}
     }}
    Scale Mode=::NewEnumerator2
   ) {module=ScaleColor002}
   /SetVariables(Particles.Color003:LinearColor=&MakeLinearColorFromVectorAndFloat(
     "Vector (RGB)"=&MakeVectorFromLinearColorRGB(Color=<rgba 5 0.1 0.1 1>)
     "Float (Alpha)"=&FloatFromCurve(FloatCurve=DataInterface() {propertyValues {Curve {keys [<at 0 1>,<at 1 0>]}}}, Scale Curve=2)
    )) {isSetParameters=true, module=SetVariables_#3}
   &ScaleSpriteSize(
    Uniform Curve Scale=1.9
    Uniform Curve Sprite Scale=DataInterface() {propertyValues {Curve {keys [{arriveTangent=1.4685, interpMode=RCIM_Cubic, leaveTangent=1.4685, tangentMode=RCTM_User, time=0, value=0}, <at 1 1 RCIM_Cubic>]}}}
   ) {module=ScaleSpriteSize001}
   &ScaleSpriteSize(
    Scale Sprite Size Mode=::NewEnumerator0
    Uniform Scale Factor=&MakeCustomFloatFromBool(Bool=&SetBoolByFloatComparison(A=1, B=$User.ParaGlideLeanAlpha), True Float=0, False Float=1)
   )
   /SetVariables(Particles.size002:Vector2f=&Multiply_Vector2D(A=$Particles.SpriteSize, B=&Vector2DFromFloat(Value=0.9))) {isSetParameters=true}
   /SetVariables(Particles.Size003:Vector2f=&Multiply_Vector2D(A=$Particles.SpriteSize, B=&Vector2DFromFloat(Value=1.9))) {isSetParameters=true, module=SetVariables_#4}
   &DynamicMaterialParameters(
    Index 0 Param 1=&FloatFromCurve(FloatCurve=DataInterface() {propertyValues {Curve {keys [<at 0.5 0>,<at 1 1>]}}})
    Module.Param0WriteEnabled=true
    Module.Param1WriteEnabled=true
    Module.Param2WriteEnabled=true
    Module.Param3WriteEnabled=true
   )
  ]
  rendererClasses [&NiagaraSpriteRendererProperties]
  renderers [
   {rendererClass=&NiagaraSpriteRendererProperties}
   {rendererClass=&NiagaraSpriteRendererProperties, rendererIndex=1}
   {rendererClass=&NiagaraSpriteRendererProperties, rendererIndex=2}
  ]
 }
 Blue_Line {
  emitterUpdateScript [&EmitterState(Life Cycle Mode=::NewEnumerator1, Loop Duration=1, UseLoopDelay=false)]
  particleSpawnScript [
   &InitializeParticle(
    Lifetime=0.03
    Sprite Size Max=<1 40>
    Sprite Size Min=<1 30>
    Sprite Size Mode=::NewEnumerator2
    UsePositionOffset=false
    Write Material Random=true
   )
   &AddVelocity(Velocity=<0 50 -200>, Velocity Speed Scale=-10)
  ]
  particleUpdateScript [
   &ParticleState()
   &ScaleSpriteSize(
    Scale Sprite Size Mode=::NewEnumerator0
    Uniform Scale Factor=&MakeCustomFloatFromBool(Bool=&SetBoolByFloatComparison(A=1, B=$User.ParaGlideLeanAlpha), True Float=0, False Float=1)
   ) {module=ScaleSpriteSize001}
   /SetVariables(Particles.Size003:Vector2f=&Multiply_Vector2D(A=$Particles.SpriteSize, B=<7 1.4>)) {isSetParameters=true}
   /SetVariables(Particles.Color003:LinearColor=<rgba 0.0365 0.2292 1 0.8>) {isSetParameters=true}
   &SolveForcesAndVelocity(Clamp Velocity=false, Limit Acceleration=false)
  ]
  rendererClasses [&NiagaraSpriteRendererProperties]
  renderers [
   {rendererClass=&NiagaraSpriteRendererProperties}
   {rendererClass=&NiagaraSpriteRendererProperties, rendererIndex=1}
  ]
 }
 Blue_Plane001 {
  emitterSpawnScript [
   /SetVariables(Emitter.Ship_InsideColor:Vector3f=&MakeVectorFromLinearColorRGB(Color=&MakeCustomLinearColorFromBool(Bool=$User.bIsFrontEndPreview, True Color=<rgba 0.6 0.0276 0.0188 1>, False Color=<rgba 2 0.05 0.02 1>))) {isSetParameters=true}
   /SetVariables(Emitter.Ship_OutsideColor:Vector3f=&MakeVectorFromLinearColorRGB(Color=&MakeCustomLinearColorFromBool(Bool=$User.bIsFrontEndPreview, True Color=<rgba 0.8 0.4834 0.2447 1>, False Color=<rgba 0.8 0.0875 0.05 1>))) {isSetParameters=true}
   /SetVariables(Emitter.FresnelExpon:NiagaraFloat=&MakeCustomFloatFromBool(Bool=$User.bIsFrontEndPreview, True Float=1.3, False Float=0.7)) {isSetParameters=true}
   /SetVariables(Emitter.Fresnelref:NiagaraFloat=&MakeCustomFloatFromBool(Bool=$User.bIsFrontEndPreview, True Float=0, False Float=0.09)) {isSetParameters=true}
  ]
  emitterUpdateScript [
   &EmitterState(Life Cycle Mode=::NewEnumerator1, Loop Duration=2, UseLoopDelay=false)
   &SpawnRate(
    SpawnRate=&MakeCustomFloatFromBool(
     Bool=&SetBoolByFloatComparison(A=1, B=$User.ParaGlideLeanAlpha)
     True Float=2
     False Float=&MakeCustomFloatFromBool(Bool=$User.bIsFrontEndPreview, True Float=8, False Float=12)
    )
    Use Spawn Probability=false
   )
   &SpawnBurst_Instantaneous(Spawn Count=5, Use Spawn Probability=false, UseLoopCountLimit=true)
  ]
  particleSpawnScript [
   &InitializeParticle(
    Color Mode=::NewEnumerator0
    Lifetime=&MakeCustomFloatFromBool(Bool=&SetBoolByFloatComparison(A=1, B=$User.ParaGlideLeanAlpha), True Float=3, False Float=0.5)
    Mesh Scale Mode=::NewEnumerator1
    Mesh Uniform Scale Max=0.12
    Mesh Uniform Scale Min=0.1
    UsePositionOffset=false
    Write Material Random=true
   )
   &ShapeLocation(
    Handle Radius=15
    Large Radius=56
    MaskSpawnGroup=false
    Offset=&MakeCustomVectorFromBool(Bool=$User.bIsFrontEndPreview, True Vector=<50 0 60>, False Vector=<40 0 100>)
    Offset Mode=::NewEnumerator0
    Override Seed=false
    Rotation Angle=0.29
    Rotation Axis=<0 0 1>
    Rotation Mode=::NewEnumerator1
    Shape Primitive=::NewEnumerator3
    U Distribution=0.4
    UseOwnerScale=true
   )
   &InitialMeshOrientation(Mesh Orientation Mode=::NewEnumerator2, Rotation=<-0.15 0.15 -0.23>, Use Rotation Vector=true)
   /SetVariables() {isSetParameters=true}
  ]
  particleUpdateScript [
   &SolveForcesAndVelocity(Clamp Velocity=false, Limit Acceleration=false)
   &ParticleState()
   &SkeletalMeshLocation(
    Bone / Socket Index=2
    MaskSpawnGroup=false
    Mesh Sampling Type=::NewEnumerator5
    Position Sampling=::NewEnumerator1
    Sampled Position Offset=$Particles.ShapeLocation.ShapePosition
    Skeletal Mesh=$User.SkeletalMesh
    Socket Sampling Mode=::NewEnumerator5
    Source Movement Scale:Vector3f=<0.8 0.8 0.8>
   )
   &ScaleMeshSize(Scale Factor=&Multiply_Vector(
     A=<0.8 1 2.1>
     B=&VectorFromFloat(Value=&MakeFloatFromVector(
       Vector=&VectorFromFloat(Value=&FloatFromCurve(
         FloatCurve=DataInterface() {propertyValues {Curve {keys [
             <at 0 0 RCIM_Cubic>
             {arriveTangent=0.0186, interpMode=RCIM_Cubic, leaveTangent=0.0186, tangentMode=RCTM_User, time=0.17, value=0.9903}
             {arriveTangent=0.049, interpMode=RCIM_Cubic, leaveTangent=0.049, tangentMode=RCTM_User, time=0.9, value=1}
             <at 1 0 RCIM_Cubic>
            ]}}}
         Scale Curve=&MakeCustomFloatFromBool(Bool=$User.bIsFrontEndPreview, True Float=1.1, False Float=1.2)
        ))
       Channel=::NewEnumerator0
      ))
    ))
   /SetVariables(Particles.Color003:LinearColor=&MakeLinearColorFromVectorAndFloat(
     "Vector (RGB)"=&MakeVectorFromLinearColorRGB(Color=<rgba 2 0.05 0.05 0.5>)
     "Float (Alpha)"=&FloatFromCurve(
      FloatCurve=DataInterface() {propertyValues {Curve {keys [
          <at 0 0 RCIM_Cubic>
          {arriveTangent=0.0186, interpMode=RCIM_Cubic, leaveTangent=0.0186, tangentMode=RCTM_User, time=0.17, value=0.9903}
          {arriveTangent=0.049, interpMode=RCIM_Cubic, leaveTangent=0.049, tangentMode=RCTM_User, time=0.9, value=1}
          <at 1 0 RCIM_Cubic>
         ]}}}
      Scale Curve=&MakeCustomFloatFromBool(Bool=$User.bIsFrontEndPreview, True Float=0.13, False Float=0.25)
     )
    )) {isSetParameters=true}
   /SetVariables(Particles.Size003:Vector2f=&Vector2DFromFloat(Value=&Multiply_Float(
      A=18
      B=&FloatFromCurve(FloatCurve=DataInterface() {propertyValues {Curve {keys [
           <at 0 0 RCIM_Cubic>
           {arriveTangent=0.0186, interpMode=RCIM_Cubic, leaveTangent=0.0186, tangentMode=RCTM_User, time=0.17, value=0.9903}
           {arriveTangent=0.049, interpMode=RCIM_Cubic, leaveTangent=0.049, tangentMode=RCTM_User, time=0.9, value=1}
           <at 1 0 RCIM_Cubic>
          ]}}})
     ))) {isSetParameters=true}
   /SetVariables(Particles.Position=&Add_Vector(
     A=Linked() {linkedVariable {name=Particles.Position, type=NiagaraPosition}}
     B=&MakeVector(
      X=&Sine(
       Normalized Angle=$Emitter.Age
       Period=&RandomRangeFloat(Minimum=2, Maximum=3)
       Scale=&RandomRangeFloat(Minimum=-1, Maximum=1)
       Bias=0
      )
      Y=0
      Z=&Sine(
       Normalized Angle=$Emitter.Age
       Period=&RandomRangeFloat(Minimum=2, Maximum=3)
       Scale=&RandomRangeFloat(Minimum=-1, Maximum=1)
       Bias=0
      )
     )
    )) {isSetParameters=true, module=SetVariables_#5}
   &ScaleMeshSize(Scale Factor=&VectorFromFloat(Value=&MakeCustomFloatFromBool(Bool=&SetBoolByFloatComparison(A=1, B=1), True Float=0, False Float=1))) {module=ScaleMeshSize001}
   &GenerateLocationEvent(Event Send Rate=5, Use Event Probability=false, UseEventDelay=false)
   &GenerateDeathEvent(Event Probability=1, Use Event Probability=true)
  ]
  rendererClasses [&NiagaraMeshRendererProperties, &NiagaraSpriteRendererProperties]
  renderers [
   {rendererClass=&NiagaraMeshRendererProperties, rendererIndex=0}
   {rendererClass=&NiagaraSpriteRendererProperties, rendererIndex=1}
  ]
 }
 Decal {
  emitterUpdateScript [
   &EmitterState(Life Cycle Mode=::NewEnumerator1, Loop Behavior=::NewEnumerator1, Loop Duration Mode=::NewEnumerator1, UseLoopDelay=false)
   &SpawnBurst_Instantaneous(Spawn Count=4, Use Spawn Probability=false, UseLoopCountLimit=false)
  ]
  particleSpawnScript [
   &PartitionParticles()
   &InitializeParticle(Color Mode=::NewEnumerator0, Sprite Size Mode=::NewEnumerator3, Uniform Sprite Size=30, UsePositionOffset=false, Write Material Random=true)
   &FN_Sprite_Attributes(
    Camera Offset=-10
    Color=&MakeCustomLinearColorFromBool(Bool=$User.bIsFrontEndPreview, True Color=<rgba 1 0.5013 0.0012 0.1>, False Color=<rgba 1 0.5013 0.0012 0.2>)
    Create Camera Offset Attribute=true
    Create Color Attribute=true
    Create SpriteSize Attribute=true
    SpriteSize=&Vector2DFromFloat(Value=50)
    Write to Module Namespace=true
   )
   &CameraOffset(Camera Offset Amount=20)
  ]
  particleUpdateScript [
   /SetVariables(Particles.LightRadius=&Multiply_Float(A=40, B=&FloatFromCurve(FloatCurve=DataInterface() {propertyValues {Curve {keys [<at 0 0>,<at 1 1>]}}}))) {isSetParameters=true}
   &ParticleState(Kill Particles When Lifetime Has Elapsed=false, Let Infinitely Lived Particles Die When Emitter Deactivates=true)
   &SkeletalMeshLocation(
    Bone / Socket Index=$Particles.PartitionParticles.Partition
    MaskSpawnGroup=false
    Mesh Sampling Type=::NewEnumerator5
    Sampled Position Offset=&SelectVectorFromArray(
     Direct Array Index=$Particles.PartitionParticles.Partition
     Vector Selection Array=DataInterface() {propertyValues {FloatData [<10 0 0>, <10 0 0>]}}
    )
    Skeletal Mesh=$User.SkeletalMesh
    Socket Sampling Mode=::NewEnumerator5
   )
   &ScaleColor(
    Scale Alpha=&Multiply_Float(
     A=&FloatFromCurve(FloatCurve=DataInterface() {propertyValues {Curve {keys [<at 0 0>,<at 0.2317 1>,<at 1 1>]}}})
     B=1
    )
    Scale RGB=&MakeVectorFromLinearColorRGB(Color=<rgba 2 0.8 0.0585 1>)
    ScaleA=true
    ScaleRGB=true
   )
  ]
  rendererClasses [&NiagaraSpriteRendererProperties]
  renderers [
   {rendererClass=&NiagaraSpriteRendererProperties}
   {rendererClass=&NiagaraSpriteRendererProperties, rendererIndex=1}
  ]
 }
 Pink_Die {
  emitterUpdateScript [&EmitterState(Life Cycle Mode=::NewEnumerator1, Loop Duration=1, UseLoopDelay=false)]
  particleSpawnScript [
   &InitializeParticle(
    Color=<rgba 1 0.4 0 1>
    Lifetime Max=0.2
    Lifetime Min=0.15
    Lifetime Mode=::NewEnumerator1
    Sprite Rotation Mode=::NewEnumerator2
    Sprite Size Mode=::NewEnumerator3
    Uniform Sprite Size=100
    UsePositionOffset=false
    Write Material Random=true
   )
   &SubUVAnimation(End Frame=3, SubUV Animation Mode:ENiagara_SubUVLookupMode=::NewEnumerator2)
  ]
  particleUpdateScript [
   &ParticleState()
   &ScaleColor(
    Linear Color Curve=DataInterface() {propertyValues {
      AlphaCurve {keys [<at 0.3624 1>,<at 0.8992 0>]}
      BlueCurve {keys [<at 0.1978 1>]}
      GreenCurve {keys [<at 0.1978 1>]}
      RedCurve {keys [<at 0.1978 1>]}
     }}
    Scale Mode=::NewEnumerator2
   ) {module=ScaleColor002}
   /SetVariables(Particles.Color003:LinearColor=&MakeLinearColorFromVectorAndFloat(
     "Vector (RGB)"=&MakeVectorFromLinearColorRGB(Color=<rgba 2 0.7514 0 1>)
     "Float (Alpha)"=&FloatFromCurve(FloatCurve=DataInterface() {propertyValues {Curve {keys [<at 0 1>,<at 1 0>]}}}, Scale Curve=2)
    )) {isSetParameters=true, module=SetVariables_#3}
   &ScaleSpriteSize(
    Uniform Curve Scale=1.7
    Uniform Curve Sprite Scale=DataInterface() {propertyValues {Curve {keys [{arriveTangent=1.4685, interpMode=RCIM_Cubic, leaveTangent=1.4685, tangentMode=RCTM_User, time=0, value=0}, <at 1 1 RCIM_Cubic>]}}}
   ) {module=ScaleSpriteSize001}
   &ScaleSpriteSize(
    Scale Sprite Size Mode=::NewEnumerator0
    Uniform Scale Factor=&MakeCustomFloatFromBool(Bool=&SetBoolByFloatComparison(A=1, B=$User.ParaGlideLeanAlpha), True Float=0, False Float=1)
   )
   /SetVariables(Particles.size002:Vector2f=&Multiply_Vector2D(A=$Particles.SpriteSize, B=&Vector2DFromFloat(Value=0.85))) {isSetParameters=true}
   /SetVariables(Particles.Size003:Vector2f=&Multiply_Vector2D(A=$Particles.SpriteSize, B=&Vector2DFromFloat(Value=1.9))) {isSetParameters=true, module=SetVariables_#4}
   &DynamicMaterialParameters(
    Index 0 Param 1=&FloatFromCurve(FloatCurve=DataInterface() {propertyValues {Curve {keys [<at 0.5 0>,<at 1 1>]}}})
    Module.Param0WriteEnabled=true
    Module.Param1WriteEnabled=true
    Module.Param2WriteEnabled=true
    Module.Param3WriteEnabled=true
   )
  ]
  rendererClasses [&NiagaraSpriteRendererProperties]
  renderers [
   {rendererClass=&NiagaraSpriteRendererProperties}
   {rendererClass=&NiagaraSpriteRendererProperties, rendererIndex=1}
   {rendererClass=&NiagaraSpriteRendererProperties, rendererIndex=2}
  ]
 }
 Pink_Line {
  emitterUpdateScript [&EmitterState(Life Cycle Mode=::NewEnumerator1, Loop Duration=1, UseLoopDelay=false)]
  particleSpawnScript [
   &InitializeParticle(
    Lifetime=0.03
    Sprite Size Max=<1 60>
    Sprite Size Min=<1 50>
    Sprite Size Mode=::NewEnumerator2
    UsePositionOffset=false
    Write Material Random=true
   )
   &AddVelocity(Velocity=<0 50 -200>, Velocity Speed Scale=10)
  ]
  particleUpdateScript [
   &ParticleState()
   &ScaleSpriteSize(
    Scale Sprite Size Mode=::NewEnumerator0
    Uniform Scale Factor=&MakeCustomFloatFromBool(Bool=&SetBoolByFloatComparison(A=1, B=$User.ParaGlideLeanAlpha), True Float=0, False Float=1)
   ) {module=ScaleSpriteSize001}
   /SetVariables(Particles.Size003:Vector2f=&Multiply_Vector2D(A=$Particles.SpriteSize, B=<7 1.4>)) {isSetParameters=true}
   /SetVariables(Particles.Color003:LinearColor=<rgba 0.0365 0.2292 1 0.8>) {isSetParameters=true}
   &SolveForcesAndVelocity(Clamp Velocity=false, Limit Acceleration=false)
  ]
  rendererClasses [&NiagaraSpriteRendererProperties]
  renderers [
   {rendererClass=&NiagaraSpriteRendererProperties}
   {rendererClass=&NiagaraSpriteRendererProperties, rendererIndex=1}
  ]
 }
 Pink_Plane001 {
  emitterSpawnScript [
   /SetVariables(Emitter.Ship_InsideColor:Vector3f=&MakeVectorFromLinearColorRGB(Color=&MakeCustomLinearColorFromBool(Bool=$User.bIsFrontEndPreview, True Color=<rgba 0.35 0.203 0 1>, False Color=<rgba 1.5 0.55 0.05 1>))) {isSetParameters=true}
   /SetVariables(Emitter.Ship_OutsideColor:Vector3f=&MakeVectorFromLinearColorRGB(Color=&MakeCustomLinearColorFromBool(Bool=$User.bIsFrontEndPreview, True Color=<rgba 2 0.8247 0 1>, False Color=<rgba 1 0.2 0 1>))) {isSetParameters=true}
   /SetVariables(Emitter.FresnelExpon:NiagaraFloat=&MakeCustomFloatFromBool(Bool=$User.bIsFrontEndPreview, True Float=1, False Float=0.7)) {isSetParameters=true}
   /SetVariables(Emitter.Fresnelref:NiagaraFloat=&MakeCustomFloatFromBool(Bool=$User.bIsFrontEndPreview, True Float=0, False Float=0.09)) {isSetParameters=true}
  ]
  emitterUpdateScript [
   &EmitterState(Life Cycle Mode=::NewEnumerator1, Loop Duration=2, UseLoopDelay=false)
   &SpawnRate(
    SpawnRate=&MakeCustomFloatFromBool(Bool=&SetBoolByFloatComparison(A=1, B=$User.ParaGlideLeanAlpha), True Float=2, False Float=5)
    Use Spawn Probability=false
   )
   &SpawnBurst_Instantaneous(Spawn Count=5, Use Spawn Probability=false, UseLoopCountLimit=true)
  ]
  particleSpawnScript [
   &InitializeParticle(
    Color Mode=::NewEnumerator0
    Lifetime=&MakeCustomFloatFromBool(Bool=&SetBoolByFloatComparison(A=1, B=$User.ParaGlideLeanAlpha), True Float=2.5, False Float=1.5)
    Mesh Scale Mode=::NewEnumerator1
    Mesh Uniform Scale Max=0.1
    Mesh Uniform Scale Min=0.09
    UsePositionOffset=false
    Write Material Random=true
   )
   &ShapeLocation(
    Handle Radius=10
    Large Radius=60
    MaskSpawnGroup=false
    Offset=&MakeCustomVectorFromBool(Bool=$User.bIsFrontEndPreview, True Vector=<30 0 165>, False Vector=<30 0 196>)
    Offset Mode=::NewEnumerator0
    Override Seed=false
    Rotation Angle=0.37
    Rotation Axis=<0 0 1>
    Rotation Mode=::NewEnumerator1
    Shape Primitive=::NewEnumerator3
    U Distribution=0.27
    UseOwnerScale=true
   )
   &InitialMeshOrientation(Mesh Orientation Mode=::NewEnumerator2, Rotation=<0.15 0.15 0.22>, Use Rotation Vector=true)
   &InitialMeshOrientation(Mesh Orientation Mode=::NewEnumerator2, Rotation=<0 -0.01 0>, Use Rotation Vector=true) {module=InitialMeshOrientation001}
   /SetVariables(Particles.CameraOffset002=-5) {isSetParameters=true}
   &CameraOffset(Camera Offset Amount=-20)
  ]
  particleUpdateScript [
   &SolveForcesAndVelocity(Clamp Velocity=false, Limit Acceleration=false)
   &ParticleState()
   &SkeletalMeshLocation(
    Bone / Socket Index=2
    MaskSpawnGroup=false
    Mesh Sampling Type=::NewEnumerator5
    Position Sampling=::NewEnumerator1
    Sampled Position Offset=$Particles.ShapeLocation.ShapePosition
    Skeletal Mesh=$User.SkeletalMesh
    Socket Sampling Mode=::NewEnumerator5
    Source Movement Scale:Vector3f=<0.8 0.8 0.8>
   )
   &ScaleMeshSize(Scale Factor=&Multiply_Vector(
     A=<0.8 1 2.1>
     B=&VectorFromFloat(Value=&FloatFromCurve(
       FloatCurve=DataInterface() {propertyValues {Curve {keys [
           <at 0 0 RCIM_Cubic>
           {arriveTangent=0.0186, interpMode=RCIM_Cubic, leaveTangent=0.0186, tangentMode=RCTM_User, time=0.17, value=0.9903}
           {arriveTangent=0.049, interpMode=RCIM_Cubic, leaveTangent=0.049, tangentMode=RCTM_User, time=0.9, value=1}
           <at 1 0 RCIM_Cubic>
          ]}}}
       Scale Curve=1.28
      ))
    ))
   /SetVariables(Particles.Color003:LinearColor=&MakeLinearColorFromVectorAndFloat(
     "Vector (RGB)"=&MakeVectorFromLinearColorRGB(Color=<rgba 1 0.4648 0.0635 0.5>)
     "Float (Alpha)"=&FloatFromCurve(
      FloatCurve=DataInterface() {propertyValues {Curve {keys [
          <at 0 0 RCIM_Cubic>
          {arriveTangent=0.0186, interpMode=RCIM_Cubic, leaveTangent=0.0186, tangentMode=RCTM_User, time=0.17, value=0.9903}
          {arriveTangent=0.049, interpMode=RCIM_Cubic, leaveTangent=0.049, tangentMode=RCTM_User, time=0.9, value=1}
          <at 1 0 RCIM_Cubic>
         ]}}}
      Scale Curve=&MakeCustomFloatFromBool(Bool=$User.bIsFrontEndPreview, True Float=0.15, False Float=0.25)
     )
    )) {isSetParameters=true}
   /SetVariables(Particles.Size003:Vector2f=&Vector2DFromFloat(Value=&Multiply_Float(
      A=20
      B=&FloatFromCurve(FloatCurve=DataInterface() {propertyValues {Curve {keys [
           <at 0 0 RCIM_Cubic>
           {arriveTangent=0.0186, interpMode=RCIM_Cubic, leaveTangent=0.0186, tangentMode=RCTM_User, time=0.17, value=0.9903}
           {arriveTangent=0.049, interpMode=RCIM_Cubic, leaveTangent=0.049, tangentMode=RCTM_User, time=0.9, value=1}
           <at 1 0 RCIM_Cubic>
          ]}}})
     ))) {isSetParameters=true}
   /SetVariables(Particles.Position=&Add_Vector(
     A=Linked() {linkedVariable {name=Particles.Position, type=NiagaraPosition}}
     B=&MakeVector(
      X=&Sine(
       Normalized Angle=$Emitter.Age
       Period=&RandomRangeFloat(Minimum=2, Maximum=3)
       Scale=&RandomRangeFloat(Minimum=-0.7, Maximum=0.7)
       Bias=0
      )
      Y=0
      Z=&Sine(
       Normalized Angle=$Emitter.Age
       Period=&RandomRangeFloat(Minimum=2, Maximum=3)
       Scale=&RandomRangeFloat(Minimum=-0.7, Maximum=0.7)
       Bias=0
      )
     )
    )) {isSetParameters=true, module=SetVariables_#5}
   &ScaleMeshSize(Scale Factor=&VectorFromFloat(Value=&MakeCustomFloatFromBool(Bool=&SetBoolByFloatComparison(A=1, B=1), True Float=0, False Float=1))) {module=ScaleMeshSize001}
   &GenerateLocationEvent(Event Send Rate=5, Use Event Probability=false, UseEventDelay=false)
   &GenerateDeathEvent(Event Probability=1, Use Event Probability=true)
  ]
  rendererClasses [&NiagaraMeshRendererProperties, &NiagaraSpriteRendererProperties]
  renderers [
   {rendererClass=&NiagaraMeshRendererProperties, rendererIndex=0}
   {rendererClass=&NiagaraSpriteRendererProperties, rendererIndex=1}
  ]
 }
 Ribbon_L {
  emitterUpdateScript [
   &EmitterState(
    Inactive Response=::NewEnumerator1
    Life Cycle Mode=::NewEnumerator1
    Loop Behavior=::NewEnumerator1
    Loop Duration Mode=::NewEnumerator1
    UseLoopDelay=false
   )
   &SpawnBurst_Instantaneous(Spawn Count=5, Use Spawn Probability=false, UseLoopCountLimit=false)
   &SpawnBurst_Instantaneous(Spawn Count=5, Spawn Group=1, Use Spawn Probability=false, UseLoopCountLimit=false) {enabled=false, module=SpawnBurst_Instantaneous001}
  ]
  particleSpawnScript [
   /SetVariables(Particles.RibbonLinkOrder:NiagaraFloat=&ReturnNormalizedExecIndex()) {isSetParameters=true, module=SetVariables_#6}
   &InitializeParticle(
    Color=&MakeLinearColorFromVectorAndFloat(
     "Vector (RGB)"=<0.7 0.7 0.7>
     "Float (Alpha)"=&FloatFromCurve(
      FloatCurve=DataInterface() {propertyValues {Curve {keys [
          {tangentMode=RCTM_User, time=0, value=1}
          {arriveTangent=-0.0405, interpMode=RCIM_Cubic, leaveTangent=-0.0405, tangentMode=RCTM_User, time=0.2, value=1}
          {arriveTangent=-2.5239, interpMode=RCIM_Cubic, leaveTangent=-2.5239, tangentMode=RCTM_User, time=1, value=0}
         ]}}}
      CurveIndex=$Particles.RibbonLinkOrder
     )
    )
    Ribbon Width=&FloatFromCurve(
     FloatCurve=DataInterface() {propertyValues {Curve {keys [<at 0 1>,<at 1 2>]}}}
     CurveIndex=$Particles.RibbonLinkOrder
     Scale Curve=&MakeCustomFloatFromBool(Bool=&SetBoolByFloatComparison(A=1, B=$User.ParaGlideLeanAlpha), True Float=16, False Float=18)
    )
    Ribbon Width Mode=::NewEnumerator1
    UsePositionOffset=false
    Write Material Random=true
   )
   /SetVariables(Particles.RibbonWidth002:NiagaraFloat=&FloatFromCurve(
     FloatCurve=DataInterface() {propertyValues {Curve {keys [<at 0 1>,<at 1 1.8>]}}}
     CurveIndex=$StackContext.RibbonLinkOrder
     Scale Curve=&MakeCustomFloatFromBool(Bool=&SetBoolByFloatComparison(A=1, B=$User.ParaGlideLeanAlpha), True Float=25, False Float=30)
    )) {isSetParameters=true}
  ]
  particleUpdateScript [
   &ParticleState(Kill Particles When Lifetime Has Elapsed=false, Let Infinitely Lived Particles Die When Emitter Deactivates=true)
   &SkeletalMeshLocation(
    "Interpolate Mesh Sampling (Update)"=true
    MaskSpawnGroup=false
    Mesh Sampling Type=::NewEnumerator5
    Position Sampling=::NewEnumerator2
    Sampled Position Offset=<10 0 0>
    Skeletal Mesh=$User.SkeletalMesh
    Socket Sampling Mode=::NewEnumerator5
   )
   /SetVariables(Particles.Position=/Niagara/DynamicInputs/Lerp/Lerp_Position(
     A:NiagaraPosition=$Particles.SkeletalMeshLocation.SampledPosition
     B:NiagaraPosition=&Add_Vector(
      A=Linked() {linkedVariable {name=Particles.SkeletalMeshLocation.SampledPosition, type=NiagaraPosition}}
      B=&Lerp_Vector(
       A=&MakeCustomVectorFromBool(Bool=$User.bIsFrontEndPreview, True Vector=<0 0 150>, False Vector=<0 0 200>)
       B=&MakeCustomVectorFromBool(Bool=$User.bIsFrontEndPreview, True Vector=<0 0 200>, False Vector=<0 0 250>)
       Alpha=$User.ParaGlideLeanAlpha
      )
     )
     Alpha:NiagaraFloat=$Particles.RibbonLinkOrder
    )) {isSetParameters=true, module=SetVariables_#7}
   /SetVariables(Particles.Position=&Add_Vector(
     A=Linked() {linkedVariable {name=Particles.Position, type=NiagaraPosition}}
     B=&MakeVector(
      X=&Sine(
       Normalized Angle=&Add_Float(
        A=&Add_Float(A=$Particles.RibbonLinkOrder, B=&Multiply_Float(A=$Emitter.Age, B=-1.2))
        B=&Multiply_FloatByInt(Float=0.33, Integer=&SpawnGroup())
       )
       Period=2
       Scale=&Multiply_Float(A=$Particles.RibbonLinkOrder, B=30)
       Bias=0
      )
      Y=&Cosine(
       Normalized Angle=&Add_Float(
        A=&Add_Float(A=$Particles.RibbonLinkOrder, B=&Multiply_Float(A=$Emitter.Age, B=-1))
        B=&Multiply_FloatByInt(Float=0.33, Integer=&SpawnGroup())
       )
       Period=1.3
       Scale=&Multiply_Float(A=$Particles.RibbonLinkOrder, B=22)
       Bias=0
      )
      Z=0
     )
    )) {isSetParameters=true, module=SetVariables_#8}
   /SetVariables(Particles.Color003:LinearColor=&MakeLinearColorFromVectorAndFloat(
     "Vector (RGB)"=&MakeVectorFromLinearColorRGB(Color=&FNTimeOfDayColor(Night Color=<rgba 1 0.65 0.0781 1>, Morning Color=<rgba 1 0.6 0 1>, Day Color=<rgba 1 0.6 0 1>, Evening Color=<rgba 1 0.65 0.0781 1>))
     "Float (Alpha)"=&FloatFromCurve(
      FloatCurve=DataInterface() {propertyValues {Curve {keys [
          {arriveTangent=-2.0494, interpMode=RCIM_Cubic, leaveTangent=-2.0494, tangentMode=RCTM_User, time=0, value=1}
          {interpMode=RCIM_Cubic, leaveTangent=-2.5239, time=1, value=0}
         ]}}}
      CurveIndex=$Particles.RibbonLinkOrder
      Scale Curve=&MakeCustomFloatFromBool(
       Bool=$User.bIsFrontEndPreview
       True Float=&FN_PickFloatDependOnPlatform(TargetPlatformActive=0.3, TargetPlatformNoActive=0.2)
       False Float=&FN_PickFloatDependOnPlatform(TargetPlatformActive=0.8, TargetPlatformNoActive=0.6)
      )
     )
    )) {isSetParameters=true}
   &DynamicMaterialParameters(
    Index 0 Param 2=&FNTimeOfDayScalar(Night Scale=0.15, Morning Scale=0, Day Scale=0, Evening Scale=0.1)
    Index 0 Param 3=&MakeCustomFloatFromBool(Bool=&SetBoolByFloatComparison(A=1, B=$User.ParaGlideLeanAlpha), True Float=1, False Float=1.3)
    Index 0 Param 4=&Lerp_Float(A=1, B=0, Alpha=$System.InterpolateOverTime.InterpolateOverTime_Float.Moving Average)
    Index 1 Param 1=&MakeCustomFloatFromBool(Bool=&SetBoolByFloatComparison(A=1, B=$User.ParaGlideLeanAlpha), True Float=1, False Float=0)
    Module.Index0Param0Write=true
    Module.Index0Param1Write=true
    Module.Index0Param2Write=true
    Module.Index0Param3Write=true
    Module.Index1Param0Write=true
    Module.Index1Param1Write=true
    Module.Index1Param2Write=true
    Module.Index1Param3Write=true
    Write Parameter Index 1=true
   )
  ]
  rendererClasses [&NiagaraRibbonRendererProperties]
  renderers [
   {rendererClass=&NiagaraRibbonRendererProperties}
   {rendererClass=&NiagaraRibbonRendererProperties, rendererIndex=1}
  ]
 }
 Ribbon_R {
  emitterUpdateScript [
   &EmitterState(
    Inactive Response=::NewEnumerator1
    Life Cycle Mode=::NewEnumerator1
    Loop Behavior=::NewEnumerator1
    Loop Duration Mode=::NewEnumerator1
    UseLoopDelay=false
   )
   &SpawnBurst_Instantaneous(Spawn Count=5, Use Spawn Probability=false, UseLoopCountLimit=false) {enabled=false}
   &SpawnBurst_Instantaneous(Spawn Count=5, Spawn Group=1, Use Spawn Probability=false, UseLoopCountLimit=false) {module=SpawnBurst_Instantaneous001}
  ]
  particleSpawnScript [
   /SetVariables(Particles.RibbonLinkOrder:NiagaraFloat=&ReturnNormalizedExecIndex()) {isSetParameters=true, module=SetVariables_#6}
   &InitializeParticle(
    Color=&MakeLinearColorFromVectorAndFloat(
     "Vector (RGB)"=<0.7 0.7 0.7>
     "Float (Alpha)"=&FloatFromCurve(
      FloatCurve=DataInterface() {propertyValues {Curve {keys [
          {tangentMode=RCTM_User, time=0, value=1}
          {arriveTangent=-0.0405, interpMode=RCIM_Cubic, leaveTangent=-0.0405, tangentMode=RCTM_User, time=0.2, value=1}
          {arriveTangent=-2.5239, interpMode=RCIM_Cubic, leaveTangent=-2.5239, tangentMode=RCTM_User, time=1, value=0}
         ]}}}
      CurveIndex=$Particles.RibbonLinkOrder
     )
    )
    Ribbon Width=&FloatFromCurve(
     FloatCurve=DataInterface() {propertyValues {Curve {keys [<at 0 1>,<at 1 2>]}}}
     CurveIndex=$Particles.RibbonLinkOrder
     Scale Curve=&MakeCustomFloatFromBool(Bool=&SetBoolByFloatComparison(A=1, B=$User.ParaGlideLeanAlpha), True Float=16, False Float=18)
    )
    Ribbon Width Mode=::NewEnumerator1
    UsePositionOffset=false
    Write Material Random=true
   )
   /SetVariables(Particles.RibbonWidth002:NiagaraFloat=&FloatFromCurve(
     FloatCurve=DataInterface() {propertyValues {Curve {keys [<at 0 1>,<at 1 1.8>]}}}
     CurveIndex=$StackContext.RibbonLinkOrder
     Scale Curve=&MakeCustomFloatFromBool(Bool=&SetBoolByFloatComparison(A=1, B=$User.ParaGlideLeanAlpha), True Float=25, False Float=30)
    )) {isSetParameters=true}
  ]
  particleUpdateScript [
   &ParticleState(Kill Particles When Lifetime Has Elapsed=false, Let Infinitely Lived Particles Die When Emitter Deactivates=true)
   &SkeletalMeshLocation(
    Bone / Socket Index=1
    "Interpolate Mesh Sampling (Update)"=true
    MaskSpawnGroup=false
    Mesh Sampling Type=::NewEnumerator5
    Position Sampling=::NewEnumerator2
    Sampled Position Offset=<10 0 0>
    Skeletal Mesh=$User.SkeletalMesh
    Socket Sampling Mode=::NewEnumerator5
   )
   /SetVariables(Particles.Position=&Lerp_Position(
     A:NiagaraPosition=$Particles.SkeletalMeshLocation.SampledPosition
     B:NiagaraPosition=&Add_Vector(
      A=Linked() {linkedVariable {name=Particles.SkeletalMeshLocation.SampledPosition, type=NiagaraPosition}}
      B=&Lerp_Vector(
       A=&MakeCustomVectorFromBool(Bool=$User.bIsFrontEndPreview, True Vector=<0 0 150>, False Vector=<0 0 200>)
       B=&MakeCustomVectorFromBool(Bool=$User.bIsFrontEndPreview, True Vector=<0 0 200>, False Vector=<0 0 250>)
       Alpha=$User.ParaGlideLeanAlpha
      )
     )
     Alpha:NiagaraFloat=$Particles.RibbonLinkOrder
    )) {isSetParameters=true, module=SetVariables_#7}
   /SetVariables(Particles.Position=&Add_Vector(
     A=Linked() {linkedVariable {name=Particles.Position, type=NiagaraPosition}}
     B=&MakeVector(
      X=&Sine(
       Normalized Angle=&Add_Float(
        A=&Add_Float(A=$Particles.RibbonLinkOrder, B=&Multiply_Float(A=$Emitter.Age, B=-1.2))
        B=&Multiply_FloatByInt(Float=0.33, Integer=&SpawnGroup())
       )
       Period=2
       Scale=&Multiply_Float(A=$Particles.RibbonLinkOrder, B=30)
       Bias=0
      )
      Y=&Cosine(
       Normalized Angle=&Add_Float(
        A=&Add_Float(A=$Particles.RibbonLinkOrder, B=&Multiply_Float(A=$Emitter.Age, B=-1))
        B=&Multiply_FloatByInt(Float=0.33, Integer=&SpawnGroup())
       )
       Period=1.3
       Scale=&Multiply_Float(A=$Particles.RibbonLinkOrder, B=20)
       Bias=0
      )
      Z=0
     )
    )) {isSetParameters=true, module=SetVariables_#8}
   /SetVariables(Particles.Color003:LinearColor=&MakeLinearColorFromVectorAndFloat(
     "Vector (RGB)"=&MakeVectorFromLinearColorRGB(Color=&FNTimeOfDayColor(Night Color=<rgba 1 0.65 0.0781 1>, Morning Color=<rgba 1 0.6 0 1>, Day Color=<rgba 1 0.6 0 1>, Evening Color=<rgba 1 0.65 0.0781 1>))
     "Float (Alpha)"=&FloatFromCurve(
      FloatCurve=DataInterface() {propertyValues {Curve {keys [
          {arriveTangent=-2.0494, interpMode=RCIM_Cubic, leaveTangent=-2.0494, tangentMode=RCTM_User, time=0, value=1}
          {interpMode=RCIM_Cubic, leaveTangent=-2.5239, time=1, value=0}
         ]}}}
      CurveIndex=$Particles.RibbonLinkOrder
      Scale Curve=&MakeCustomFloatFromBool(
       Bool=$User.bIsFrontEndPreview
       True Float=&FN_PickFloatDependOnPlatform(TargetPlatformActive=0.3, TargetPlatformNoActive=0.2)
       False Float=&FN_PickFloatDependOnPlatform(TargetPlatformActive=0.8, TargetPlatformNoActive=0.6)
      )
     )
    )) {isSetParameters=true}
   &DynamicMaterialParameters(
    Index 0 Param 1=1
    Index 0 Param 2=&FNTimeOfDayScalar(Night Scale=0.15, Morning Scale=0, Day Scale=0, Evening Scale=0.1)
    Index 0 Param 3=&MakeCustomFloatFromBool(Bool=&SetBoolByFloatComparison(A=1, B=$User.ParaGlideLeanAlpha), True Float=1, False Float=1.3)
    Index 0 Param 4=&Lerp_Float(A=1, B=0, Alpha=$System.InterpolateOverTime.InterpolateOverTime_Float.Moving Average)
    Index 1 Param 1=&MakeCustomFloatFromBool(Bool=&SetBoolByFloatComparison(A=1, B=$User.ParaGlideLeanAlpha), True Float=1, False Float=0)
    Module.Index0Param0Write=true
    Module.Index0Param1Write=true
    Module.Index0Param2Write=true
    Module.Index0Param3Write=true
    Module.Index1Param0Write=true
    Module.Index1Param1Write=true
    Module.Index1Param2Write=true
    Module.Index1Param3Write=true
    Write Parameter Index 1=true
   )
  ]
  rendererClasses [&NiagaraRibbonRendererProperties]
  renderers [
   {rendererClass=&NiagaraRibbonRendererProperties}
   {rendererClass=&NiagaraRibbonRendererProperties, rendererIndex=1}
  ]
 }
}
summary {
 systemName=NS_Contrail_Dive_Retro
 emitters [
  {emitterName=Ribbon_L, rendererClasses [&NiagaraRibbonRendererProperties]}
  {emitterName=Ribbon_R, rendererClasses [&NiagaraRibbonRendererProperties]}
  {emitterName=Blue_Line, rendererClasses [&NiagaraSpriteRendererProperties]}
  {emitterName=Blue_Die, rendererClasses [&NiagaraSpriteRendererProperties]}
  {emitterName=Pink_Die, rendererClasses [&NiagaraSpriteRendererProperties]}
  {emitterName=Pink_Line, rendererClasses [&NiagaraSpriteRendererProperties]}
  {emitterName=Pink_Plane001, rendererClasses [&NiagaraMeshRendererProperties, &NiagaraSpriteRendererProperties]}
  {emitterName=Blue_Plane001, rendererClasses [&NiagaraMeshRendererProperties, &NiagaraSpriteRendererProperties]}
  {emitterName=Decal, rendererClasses [&NiagaraSpriteRendererProperties]}
 ]
 userVariables [
  User.bIsFrontEndPreview {type=NiagaraBool}
  User.ParaGlideLeanAlpha {type=NiagaraFloat, defaultValue=1, description=""}
  User.SkeletalMesh {
   type=NiagaraDataInterfaceSkeletalMesh
   defaultValue {
    struct=NiagaraExt_VariableValue_DataInterface
    value {dataInterface=&self:NiagaraDataInterfaceSkeletalMesh_0}
   }
  }
 ]
}
